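# Physics asset for the Rumble Female one-handed axe pickaxe, setting up bodies and joints for a dangling ribbon with a bow.
# A compound handle body of capsules and a box anchors the ribbon root, which carries the knot.
# The knot branches into two bow loops and left and right three-segment tails via master bones.
# Root and master joints rotate freely, while tail segments have tight twist limits and bows are stiffer.
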PhysicsAsset /BRCosmetics/Weapons/FORT_Melee/Pickaxe_Rumble_Female_1h/Meshes/Rumble_Female_1h_Axe_PhysicsAsset
bodies {
 Melee_TwoHanded_Handle [
  {shapeType=Capsule, length=45, radius=4.2, center=<-0.25 -0.5 21>}
  {shapeType=Capsule, length=46, radius=2.5, shapeName=Melee_TwoHanded_Handle_capsule, center=<1 -1 30>}
  {shapeType=Capsule, length=15, radius=5.5, shapeName=Melee_TwoHanded_Handle_capsule, center=<-0.2605 -0.7209 28>}
  {shapeType=Box, extentX=1, extentY=8.5, extentZ=7, shapeName=Melee_TwoHanded_Handle_box, center=<0.1905 0.6 52>, rotation=<rot 0 90 0>}
 ]
 dyn_ribbon_0_LFT [{shapeType=Capsule, length=4, radius=0.75, shapeName=dyn_ribbon_0_LFT_capsule, center=<3.4 0 0>, rotation=<rot 90 0 0>}]
 dyn_ribbon_0_RGT [{shapeType=Capsule, length=5, radius=0.75, shapeName=dyn_ribbon_0_RGT_capsule, center=<3 0 0>, rotation=<rot 90 0 0>}]
 dyn_ribbon_1_LFT [{shapeType=Capsule, length=5, radius=1.25, shapeName=dyn_ribbon_1_LFT_capsule, center=<3 0 0>, rotation=<rot 90 0 0>}]
 dyn_ribbon_1_RGT [{shapeType=Capsule, length=5, radius=1.25, shapeName=dyn_ribbon_1_RGT_capsule, center=<3.4 0 0>, rotation=<rot 90 0 0>}]
 dyn_ribbon_2_LFT [{shapeType=Capsule, length=6, radius=1.4, shapeName=dyn_ribbon_2_LFT_capsule, center=<4 0 0>, rotation=<rot 90 0 0>}]
 dyn_ribbon_2_RGT [{shapeType=Capsule, length=6, radius=1.4, shapeName=dyn_ribbon_2_RGT_capsule, center=<4 0 0>, rotation=<rot 90 0 0>}]
 dyn_ribbon_LFT_MSTR [{shapeType=Capsule, length=0.505, radius=0.505, shapeName=dyn_ribbon_LFT_MSTR_capsule, rotation=<rot 0 -90 0>}]
 dyn_ribbon_RGT_MSTR [{shapeType=Capsule, length=0.505, radius=0.505, shapeName=dyn_ribbon_RGT_MSTR_capsule, rotation=<rot 0 -90 0>}]
 dyn_ribbon_bow_LFT [{shapeType=Capsule, length=1.5, radius=1.5, shapeName=dyn_ribbon_bow_LFT_capsule, center=<1.6 0 0>, rotation=<rot 90 0 0>}]
 dyn_ribbon_bow_RGT [{shapeType=Capsule, length=1.5, radius=1.5, shapeName=dyn_ribbon_bow_RGT_capsule, center=<1.6 0 0>, rotation=<rot 90 0 0>}]
 dyn_ribbon_knot [{shapeType=Capsule, length=0.75, radius=1.25, shapeName=dyn_ribbon_knot_capsule}]
 dyn_ribbon_root [{shapeType=Sphere, radius=0.5, shapeName=dyn_ribbon_root_sphere}]
}
constraints [
 {bone1Name=dyn_ribbon_knot, bone2Name=dyn_ribbon_root, swing1LimitDegrees=45, swing2LimitDegrees=45, twistLimitDegrees=45}
 {
  bone1Name=dyn_ribbon_root
  bone2Name=Melee_TwoHanded_Handle
  swing1LimitDegrees=45
  swing1Motion=Free
  swing2LimitDegrees=45
  swing2Motion=Free
  twistLimitDegrees=45
  twistMotion=Free
 }
 {
  bone1Name=dyn_ribbon_RGT_MSTR
  bone2Name=dyn_ribbon_knot
  swing1LimitDegrees=45
  swing1Motion=Free
  swing2LimitDegrees=45
  swing2Motion=Free
  twistLimitDegrees=45
  twistMotion=Free
 }
 {bone1Name=dyn_ribbon_0_RGT, bone2Name=dyn_ribbon_RGT_MSTR, swing1LimitDegrees=45, swing2LimitDegrees=45, twistLimitDegrees=5}
 {bone1Name=dyn_ribbon_1_RGT, bone2Name=dyn_ribbon_0_RGT, swing1LimitDegrees=45, swing2LimitDegrees=45, twistLimitDegrees=5}
 {bone1Name=dyn_ribbon_2_RGT, bone2Name=dyn_ribbon_1_RGT, swing1LimitDegrees=45, swing2LimitDegrees=45, twistLimitDegrees=5}
 {
  bone1Name=dyn_ribbon_LFT_MSTR
  bone2Name=dyn_ribbon_knot
  swing1LimitDegrees=45
  swing1Motion=Free
  swing2LimitDegrees=45
  swing2Motion=Free
  twistLimitDegrees=45
  twistMotion=Free
 }
 {bone1Name=dyn_ribbon_0_LFT, bone2Name=dyn_ribbon_LFT_MSTR, swing1LimitDegrees=45, swing2LimitDegrees=45, twistLimitDegrees=5}
 {bone1Name=dyn_ribbon_1_LFT, bone2Name=dyn_ribbon_0_LFT, swing1LimitDegrees=45, swing2LimitDegrees=45, twistLimitDegrees=5}
 {bone1Name=dyn_ribbon_2_LFT, bone2Name=dyn_ribbon_1_LFT, swing1LimitDegrees=45, swing2LimitDegrees=45, twistLimitDegrees=5}
 {bone1Name=dyn_ribbon_bow_RGT, bone2Name=dyn_ribbon_knot, swing1LimitDegrees=30, swing2LimitDegrees=30, twistLimitDegrees=2}
 {bone1Name=dyn_ribbon_bow_LFT, bone2Name=dyn_ribbon_knot, swing1LimitDegrees=30, swing2LimitDegrees=30, twistLimitDegrees=2}
]
